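# Niagara system for the explosion effect of the CyberMitt emote.
# A debris and spark emitter splits bursts by spawn group for different size, velocity, drag and gravity, offsetting some along randomly rotated preset directions.
# Mesh smoke uses curve-driven dynamic material parameters, a global scale and time-of-day tinting.
# Flare meshes and streak sprites add the flash, alongside a curl-noise burst emitter named Souce.
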NiagaraSystem /BRCosmetics/Animation/Game/MainPlayer/Emotes/CyberMitt/VFX/NS_Emote_CyberMitt_Explosion
dependencies {
 usedDataInterfaces [
  &NiagaraDataInterfaceCurve
  &NiagaraDataInterfaceVector2DCurve
  &NiagaraDataInterfaceColorCurve
  &NiagaraDataInterfaceVectorCurve
  &NiagaraDataInterfaceParticleRead
  &NiagaraDataInterfaceArrayFloat3
 ]
 usedDynamicInputs [&ReturnNormalizedExecIndex]
 usedModules [&SystemState]
}
emitters {
 Debris_Spark_Explosion {
  emitterSpawnScript [/SetVariables(Emitter.RandomVectorRotation:NiagaraFloat=&RandomRangeFloat(Minimum=0, Maximum=360)) {isSetParameters=true}]
  emitterUpdateScript [
   &EmitterState()
   &SpawnBurst_Instantaneous(Spawn Count=&RandomRangeInt(Minimum=5, Maximum=15), Use Spawn Probability=false)
   &SpawnBurst_Instantaneous(Spawn Count=90, Spawn Group=1, Use Spawn Probability=false, UseLoopCountLimit=false) {module=SpawnBurst_Instantaneous001}
   &SpawnBurst_Instantaneous(Spawn Count=5, Spawn Group=2, Use Spawn Probability=false, UseLoopCountLimit=false) {module=SpawnBurst_Instantaneous002}
  ]
  particleSpawnScript [
   /SetVariables(Particles.VisibilityTag=&SpawnGroup()) {isSetParameters=true, module=SetVariables_#1}
   &InitializeParticle(
    Color=&MaskLinearColorBySpawnGroup(
     Masked Linear Color=<rgba 1 1 1 1>
     Passthrough Linear Color=&MaskLinearColorBySpawnGroup(
      Masked Linear Color=<rgba 1 1 1 1>
      Passthrough Linear Color=&UniformRangedLinearColor(Minimum=<rgba 1 0.8866 0.786 1>, Maximum=<rgba 0.1198 0.1032 0.0936 1>)
      Spawn Group Mask=1
     )
    )
    Lifetime=&MaskFloatBySpawnGroup(
     Masked Float=&RandomRangeFloat(Minimum=0.5, Maximum=1)
     Passthrough Float=&MaskFloatBySpawnGroup(
      Masked Float=&RandomRangeFloat(Minimum=0.5, Maximum=1)
      Passthrough Float=&RandomRangeFloat(Minimum=0.3, Maximum=0.5)
      Spawn Group Mask=1
     )
    )
    Mesh Scale Mode=::NewEnumerator1
    Mesh Uniform Scale Max=0.75
    Mesh Uniform Scale Min=0.25
    Position Offset=&MaskVectorBySpawnGroup(
     Masked Vector=&Multiply_VectorByFloat(
      Vector=/Niagara/DynamicInputs/Vectors/RotateVector(
       Vector To Rotate:Vector3f=&SelectVectorFromArray(
        Direct Array Index=&ReturnParticleIDAttributes()
        Direct Array Mode=::NewEnumerator1
        Vector Selection Array=DataInterface() {propertyValues {FloatData [<0 0.5 0>, <0.5405 -1.2953 1.0181>, <0.8 0.964 0.5>, <-1 0 0.6937>]}}
       )
       Yaw:NiagaraFloat=$Emitter.RandomVectorRotation
       Pitch:NiagaraFloat=0
       Roll:NiagaraFloat=0
       Rotation Rate:NiagaraFloat=&Add_Float(A=1, B=0)
      )
      Float=5
     )
     Passthrough Vector=&MaskVectorBySpawnGroup(
      Masked Vector=&Multiply_VectorByFloat(
       Vector=&RotateVector(
        Vector To Rotate:Vector3f=&SelectVectorFromArray(
         Direct Array Index=&ReturnParticleIDAttributes()
         Direct Array Mode=::NewEnumerator1
         Vector Selection Array=DataInterface() {propertyValues {FloatData [<0 0.5 0>, <0.5405 -1.2953 1.0181>, <0.8 0.964 0.5>, <-1 0 0.6937>]}}
        )
        Yaw:NiagaraFloat=$Emitter.RandomVectorRotation
        Pitch:NiagaraFloat=0
        Roll:NiagaraFloat=0
        Rotation Rate:NiagaraFloat=&Add_Float(A=1, B=0)
       )
       Float=5
      )
      Passthrough Vector=<0 0 0>
      Spawn Group Mask=1
     )
    )
    Sprite Size=&MaskVector2dBySpawnGroup(
     Masked Vector2D=&RandomRangeVector2D(Minimum=<0.2 1>, Maximum=<2 3>)
     Passthrough Vector2D=&MaskVector2dBySpawnGroup(
      Masked Vector2D=&RandomRangeVector2D(Minimum=<0.2 1>, Maximum=<2 3>)
      Passthrough Vector2D=&Vector2DFromFloat(Value=&RandomRangeFloat(Minimum=35, Maximum=60))
      Spawn Group Mask=1
     )
     Spawn Group Mask=0
    )
    Sprite Size Mode=::NewEnumerator4
    UsePositionOffset=true
    Write Material Random=true
   ) {module=InitializeParticle001}
   &AddVelocity(
    Velocity=&MaskVectorBySpawnGroup(
     Masked Vector=&RandomRangeVector(Minimum=<0 0 5>, Maximum=<0 0 50>)
     Passthrough Vector=&MaskVectorBySpawnGroup(
      Masked Vector=&RandomRangeVector(Minimum=<0 0 5>, Maximum=<0 0 5>)
      Passthrough Vector=<0 0 0>
      Spawn Group Mask=1
     )
    )
    Velocity Speed Scale=&RandomRangeFloat(Minimum=5, Maximum=10)
   )
   &AddVelocity(
    Cone Angle=5
    Cone Axis=&RotateVector(
     Vector To Rotate:Vector3f=&SelectVectorFromArray(
      Direct Array Index=&ReturnParticleIDAttributes()
      Direct Array Mode=::NewEnumerator1
      Vector Selection Array=DataInterface() {propertyValues {FloatData [<0 0.5 0>, <0.5405 -1.2953 1.0181>, <0.8 0.964 0.5>, <-1 0 0.6937>]}}
     )
     Yaw:NiagaraFloat=$Emitter.RandomVectorRotation
     Pitch:NiagaraFloat=0
     Roll:NiagaraFloat=0
     Rotation Rate:NiagaraFloat=&Add_Float(A=1, B=0)
    )
    Module.UseConeFalloff=false
    Module.UseVelDistribution=false
    Override Seed=false
    Velocity Mode=::NewEnumerator2
    Velocity Speed=&Multiply_Float(
     A=&RandomRangeFloat(Minimum=0, Maximum=350)
     B=&MaskFloatBySpawnGroup(
      Masked Float=1
      Passthrough Float=&MaskFloatBySpawnGroup(Masked Float=1, Passthrough Float=0, Spawn Group Mask=1)
     )
    )
   ) {module=AddVelocity001}
   &AddVelocity(
    Module.Use Falloff Distance=false
    Override Seed=false
    Velocity Mode=::NewEnumerator1
    Velocity Speed=&MaskFloatBySpawnGroup(
     Masked Float=0
     Passthrough Float=&MaskFloatBySpawnGroup(Masked Float=0, Passthrough Float=170, Spawn Group Mask=1)
    )
   ) {module=AddVelocity002}
   &InitialMeshOrientation(Use Rotation Vector=false)
  ]
  particleUpdateScript [
   &ParticleState()
   /SetVariables(
    Particles.ExplosionDyanmic_00:Vector4f=&MakeVector4(
     X=&FloatFromCurve(FloatCurve=DataInterface() {propertyValues {Curve {keys [<at 0 0.35>,<at 1 1>]}}})
     Y=0
     Z=&FloatFromCurve(FloatCurve=DataInterface() {propertyValues {Curve {keys [<at 0 10>,<at 0.4 0>]}}})
     W=0
    )
    Particles.ExplosionDyanmic_01:Vector4f=&MakeVector4(
     X=&FloatFromCurve(
      FloatCurve=DataInterface() {propertyValues {Curve {keys [
          {interpMode=RCIM_Cubic, leaveTangent=4.7088, tangentMode=RCTM_Break, time=0, value=0}
          {arriveTangent=4.2257, interpMode=RCIM_Cubic, leaveTangent=4.2257, tangentMode=RCTM_User, time=0.25, value=1}
         ]}}}
      Scale Curve=&RandomRangeFloat(Minimum=1, Maximum=1.5)
     )
     Y=&FloatFromCurve(
      FloatCurve=DataInterface() {propertyValues {Curve {keys [<at 0 0.2>, {leaveTangent=0.9066, time=1, value=2}]}}}
      Scale Curve=&RandomRangeFloat(Minimum=0.1, Maximum=0.2)
     )
     Z=&FloatFromCurve(FloatCurve=DataInterface() {propertyValues {Curve {keys [
          {interpMode=RCIM_Cubic, leaveTangent=6.3917, tangentMode=RCTM_Break, time=0, value=0.2}
          {arriveTangent=1.7549, interpMode=RCIM_Cubic, leaveTangent=10.6767, tangentMode=RCTM_User, time=0.25, value=1}
         ]}}})
     W=&FloatFromCurve(
      FloatCurve=DataInterface() {propertyValues {Curve {keys [{arriveTangent=-0.136, interpMode=RCIM_Cubic, leaveTangent=-0.136, tangentMode=RCTM_User, time=0, value=1}, <at 1 0.95>]}}}
      Scale Curve=&RandomRangeFloat(Minimum=1.2, Maximum=1.4)
     )
    )
    Particles.ExplosionDyanmic_02:Vector4f=&MakeVector4(
     X=&FloatFromCurve(FloatCurve=DataInterface() {propertyValues {Curve {keys [
          {interpMode=RCIM_Cubic, leaveTangent=4.2485, tangentMode=RCTM_Break, time=0, value=0}
          {arriveTangent=1.0548, interpMode=RCIM_Cubic, leaveTangent=1.0548, tangentMode=RCTM_User, time=0.17, value=0.5715}
          <at 1 1>
         ]}}})
     Y=&FloatFromCurve(
      FloatCurve=DataInterface() {propertyValues {Curve {keys [
          {arriveTangent=1.9056, interpMode=RCIM_Cubic, leaveTangent=1.9056, tangentMode=RCTM_User, time=0, value=0.2}
          {arriveTangent=0.0259, interpMode=RCIM_Cubic, leaveTangent=0.0259, tangentMode=RCTM_User, time=1, value=1.4}
         ]}}}
      Scale Curve=&RandomRangeFloat(Minimum=0, Maximum=0)
     )
     Z=&FloatFromCurve(FloatCurve=DataInterface() {propertyValues {Curve {keys [
          {interpMode=RCIM_Cubic, leaveTangent=2.3064, tangentMode=RCTM_Break, time=0, value=1}
          {interpMode=RCIM_Cubic, leaveTangent=11.1233, tangentMode=RCTM_Break, time=1, value=5}
         ]}}})
     W=&FloatFromCurve(
      FloatCurve=DataInterface() {propertyValues {Curve {keys [
          {arriveTangent=-0.6867, interpMode=RCIM_Cubic, leaveTangent=-0.6867, tangentMode=RCTM_User, time=0, value=1}
          {arriveTangent=-0.2589, interpMode=RCIM_Cubic, leaveTangent=-0.2589, tangentMode=RCTM_User, time=0.3, value=0.8}
          {interpMode=RCIM_Cubic, leaveTangent=-1.2956, tangentMode=RCTM_Break, time=1, value=0.75}
         ]}}}
      Scale Curve=1.6
     )
    )
   ) {isSetParameters=true}
   &SpriteRotationRate(Rotation Rate=&MaskFloatBySpawnGroup(
     Masked Float=0
     Passthrough Float=&MaskFloatBySpawnGroup(
      Masked Float=0
      Passthrough Float=&Multiply_Float(
       A=&RandomRangeFloat(Minimum=-3, Maximum=3)
       B=&FloatFromCurve(
        FloatCurve=DataInterface() {propertyValues {Curve {keys [
            {arriveTangent=-4.0107, interpMode=RCIM_Cubic, leaveTangent=-4.0107, tangentMode=RCTM_User, time=-0.0046, value=1.0113}
            {arriveTangent=-0.2496, interpMode=RCIM_Cubic, leaveTangent=-0.2496, tangentMode=RCTM_User, time=0.5, value=0.16}
           ]}}}
        CurveIndex=$Emitter.Age
        Scale Curve=150
       )
      )
      Spawn Group Mask=1
     )
    ))
   /SetVariables(Particles.SpriteSpark_Color:LinearColor=&ColorFromCurve(ColorCurve=DataInterface() {propertyValues {
       AlphaCurve {keys [<at 0 1 RCIM_Cubic>,<at 1 0 RCIM_Cubic>]}
       BlueCurve {keys [<at 0 3.5 RCIM_Cubic>,<at 0.408 0>,<at 1 0 RCIM_Cubic>]}
       GreenCurve {keys [<at 0 3.6 RCIM_Cubic>,<at 0.408 9.1>,<at 1 0 RCIM_Cubic>]}
       RedCurve {keys [<at 0 4 RCIM_Cubic>,<at 0.408 50>,<at 1 1 RCIM_Cubic>]}
      }})) {isSetParameters=true}
   &AccelerationForce(Acceleration=&MaskVectorBySpawnGroup(
     Masked Vector=<0 0 0>
     Passthrough Vector=&MaskVectorBySpawnGroup(
      Masked Vector=&RandomRangeVector(Minimum=<0 0 -200>, Maximum=<0 0 -400>)
      Passthrough Vector=<0 0 0>
      Spawn Group Mask=1
     )
    ))
   &UpdateMeshOrientation(
    Rotation Rate=&RandomRangeFloat(Minimum=0, Maximum=2)
    Rotation Vector=&RandomRangeVector(Minimum=<0 0 0>, Maximum=<1 1 1>)
   )
   &Drag(
    Drag=&MaskFloatBySpawnGroup(
     Masked Float=&RandomRangeFloat(Minimum=0.5, Maximum=2)
     Passthrough Float=&MaskFloatBySpawnGroup(
      Masked Float=&RandomRangeFloat(Minimum=0.5, Maximum=2)
      Passthrough Float=&FloatFromCurve(FloatCurve=DataInterface() {propertyValues {Curve {keys [
           {interpMode=RCIM_Cubic, leaveTangent=0.0927, tangentMode=RCTM_Break, time=0, value=1}
           {arriveTangent=-2.586, interpMode=RCIM_Cubic, leaveTangent=-2.586, tangentMode=RCTM_User, time=1, value=0}
          ]}}})
      Spawn Group Mask=1
     )
    )
    Rotational Drag:NiagaraFloat=&Lerp_Float(A=1, B=2, Alpha=$Particles.Initial.MaterialRandom)
    Use Linear Drag=true
    Use Rotational Drag=true
   )
   &GravityForce(Gravity=&MaskVectorBySpawnGroup(
     Masked Vector=&RandomRangeVector(Minimum=<0 0 -500>, Maximum=<0 0 -980>)
     Passthrough Vector=&MaskVectorBySpawnGroup(
      Masked Vector=&RandomRangeVector(Minimum=<0 0 -500>, Maximum=<0 0 -980>)
      Passthrough Vector=<0 0 0>
      Spawn Group Mask=1
     )
    ))
   &SolveForcesAndVelocity(Clamp Velocity=false, Limit Acceleration=false)
   &ScaleSpriteSizeBySpeed(
    Max Scale Factor=&MaskVector2dBySpawnGroup(
     Masked Vector2D=<1 5>
     Passthrough Vector2D=&MaskVector2dBySpawnGroup(Masked Vector2D=<1 5>, Passthrough Vector2D=<1 1>, Spawn Group Mask=1)
     Spawn Group Mask=0
    )
    Min Scale Factor=&MaskVector2dBySpawnGroup(
     Masked Vector2D=<0.65 0.65>
     Passthrough Vector2D=&MaskVector2dBySpawnGroup(Masked Vector2D=<0.65 0.65>, Passthrough Vector2D=<1 1>, Spawn Group Mask=1)
     Spawn Group Mask=0
    )
    Velocity Threshold=500
   )
   &ScaleSpriteSize(
    Scale Factor=&MaskVector2dBySpawnGroup(
     Masked Vector2D=&Vector2DFromCurve(
      Vector2Curve=DataInterface() {propertyValues {XCurve {keys [<at 0 0.5>,<at 0.2 1>,<at 1 1>]}, YCurve {keys [<at 0 3>,<at 0.2 1>,<at 1 0>]}}}
      ScaleCurve=<1 1>
     )
     Passthrough Vector2D=&MaskVector2dBySpawnGroup(
      Masked Vector2D=&Vector2DFromCurve(
       Vector2Curve=DataInterface() {propertyValues {XCurve {keys [<at 0 0.5>,<at 0.2 1>,<at 1 1>]}, YCurve {keys [<at 0 3>,<at 0.2 1>,<at 1 0>]}}}
       ScaleCurve=<1 1>
      )
      Passthrough Vector2D=&Vector2DFromFloat(Value=&FloatFromCurve(FloatCurve=DataInterface() {propertyValues {Curve {keys [
            {arriveTangent=8.1876, interpMode=RCIM_Cubic, leaveTangent=5.3162, tangentMode=RCTM_User, time=0, value=0.15}
            {arriveTangent=1.0025, leaveTangent=1.0025, time=0.1, value=1}
            {arriveTangent=0.2754, interpMode=RCIM_Cubic, leaveTangent=0.2754, tangentMode=RCTM_User, time=1, value=2}
           ]}}}))
      Spawn Group Mask=1
     )
     Spawn Group Mask=0
    )
    Scale Sprite Size Mode=::NewEnumerator2
   )
   &ScaleColor(
    Scale Alpha=&MaskFloatBySpawnGroup(
     Masked Float=1
     Passthrough Float=&MaskFloatBySpawnGroup(
      Masked Float=1
      Passthrough Float=&FloatFromCurve(FloatCurve=DataInterface() {propertyValues {Curve {keys [<at 0.1219 4>,<at 0.194 1.5>,<at 0.7612 1.5>,<at 1 0>]}}})
      Spawn Group Mask=1
     )
    )
    Scale RGB=&MaskVectorBySpawnGroup(
     Masked Vector=<1 1 1>
     Passthrough Vector=&MaskVectorBySpawnGroup(
      Masked Vector=<1 1 1>
      Passthrough Vector=&VectorFromFloat(Value=&FloatFromCurve(FloatCurve=DataInterface() {propertyValues {Curve {keys [
            {arriveTangent=1.8145, interpMode=RCIM_Cubic, leaveTangent=4.5795, tangentMode=RCTM_Break, time=0, value=0.4}
            <at 0.3387 0.9565>
            <at 1 0.7>
           ]}}}))
      Spawn Group Mask=1
     )
    )
    ScaleA=true
    ScaleRGB=true
   )
   &CameraOffset(Camera Offset Amount=&MaskFloatBySpawnGroup(
     Masked Float=0
     Passthrough Float=&MaskFloatBySpawnGroup(Masked Float=0, Passthrough Float=35, Spawn Group Mask=1)
    ))
  ]
  rendererClasses [&NiagaraMeshRendererProperties, &NiagaraSpriteRendererProperties]
  renderers [
   {rendererClass=&NiagaraMeshRendererProperties, rendererIndex=0}
   {rendererClass=&NiagaraSpriteRendererProperties, rendererIndex=1}
   {rendererClass=&NiagaraSpriteRendererProperties, rendererIndex=2}
  ]
 }
 Main_Smoke {
  emitterSpawnScript [
   /SetVariables(
    Emitter.Global Color:LinearColor=<rgba 4 0.349 0 1>
    Emitter.Global Scale:NiagaraFloat=&Multiply_Float(A=0.15, B=1)
   ) {isSetParameters=true}
   /SetVariables(
    Emitter.Color2:LinearColor=&MultiplyLinearColor(Linear Color=$Emitter.Global Color, Scale Factor=<2 1 1 1>)
    Emitter.Depth Fade:NiagaraFloat=&Multiply_Float(A=125, B=$Emitter.Global Scale)
    Emitter.Dissolve Noise Power=1.1
    Emitter.Dissolve Noise Scale:LinearColor=&MultiplyLinearColor(
     Linear Color=/Niagara/DynamicInputs/TypeConversions/MakeLinearColorFromFloat(Value:NiagaraFloat=&FloatFromCurve(FloatCurve=DataInterface() {propertyValues {Curve {keys [<at 0 2>,<at 1 1>,<at 2 0.5>]}}}, CurveIndex=$Emitter.Global Scale))
     Scale Factor=<0.0008 0.0008 0.0009 1>
    )
    Emitter.Near Camera Fade:NiagaraFloat=&Multiply_Float(A=400, B=$Emitter.Global Scale)
    Emitter.Use SkyAtmosphereLightColor=true
   ) {isSetParameters=true}
  ]
  emitterUpdateScript [
   &EmitterState(Life Cycle Mode=::NewEnumerator1, Loop Behavior=::NewEnumerator1, Loop Duration=0.627, UseLoopDelay=false)
   &SpawnBurst_Instantaneous(Spawn Count=7, Use Spawn Probability=false, UseLoopCountLimit=false) {module=SpawnBurst_Instantaneous001}
  ]
  particleSpawnScript [
   &InitializeParticle(
    Lifetime=&RandomRangeFloat(Minimum=0.45, Maximum=0.6)
    Mesh Scale Mode=::NewEnumerator3
    Mesh Uniform Scale=&RandomRangeFloat(Minimum=3, Maximum=4)
    UsePositionOffset=false
    Write Material Random=true
   )
   &ShapeLocation(MaskSpawnGroup=false, Offset Mode=::NewEnumerator0, Override Seed=false, Sphere Radius=1, UseOwnerScale=true)
   &AddVelocity(
    Module.Use Falloff Distance=false
    Origin Offset=&Multiply_VectorByFloat(Vector=&RandomRangeVector(Minimum=<0 0 0>, Maximum=<0 0 -5>), Float=$Emitter.Global Scale)
    Override Seed=false
    Velocity Mode=::NewEnumerator1
    Velocity Speed=&Multiply_Float(
     A=&MaskFloatBySpawnGroup(Masked Float=1400, Passthrough Float=&RandomRangeFloat(Minimum=500, Maximum=900), Spawn Group Mask=1)
     B=$Emitter.Global Scale
    )
   )
   &CameraOffset(Camera Offset Amount=&Multiply_Float(A=&MaskFloatBySpawnGroup(Masked Float=70, Passthrough Float=25, Spawn Group Mask=1), B=$Emitter.Global Scale))
   /SetVariables(Particles.Scale=&Multiply_VectorByFloat(Vector=$Particles.Scale, Float=$Emitter.Global Scale)) {isSetParameters=true}
  ]
  particleUpdateScript [
   &ParticleState()
   &GravityForce(Gravity=&Multiply_VectorByFloat(
     Vector=&MaskVectorBySpawnGroup(
      Masked Vector=<0 0 500>
      Passthrough Vector=&RandomRangeVector(Minimum=<0 0 400>, Maximum=<1 1 600>)
      Spawn Group Mask=1
     )
     Float=$Emitter.Global Scale
    ))
   &Drag(Drag=&RandomRangeFloat(Minimum=4, Maximum=5), Use Linear Drag=true, Use Rotational Drag=false)
   &DynamicMaterialParameters(
    Index 0 Param 1=&Multiply_Float(
     A=&MaskFloatBySpawnGroup(
      Masked Float=-150
      Passthrough Float=&RandomRangeFloat(Minimum=-90, Maximum=-120)
      Spawn Group Mask=&MakeIntFromFloat(Float=1)
     )
     B=$Emitter.Global Scale
    )
    Index 0 Param 2=&FloatFromCurve(
     FloatCurve=DataInterface() {propertyValues {Curve {keys [
         {tangentMode=RCTM_User, time=0, value=1}
         {arriveTangent=-0.0405, interpMode=RCIM_Cubic, leaveTangent=-0.0405, tangentMode=RCTM_User, time=0.4, value=1}
         {arriveTangent=-1.6592, interpMode=RCIM_Cubic, leaveTangent=-1.6592, tangentMode=RCTM_User, time=1, value=0.4}
        ]}}}
     Scale Curve=0.5
    )
    Index 0 Param 3=&FloatFromCurve(
     FloatCurve=DataInterface() {propertyValues {Curve {keys [
         {arriveTangent=1.4725, interpMode=RCIM_Cubic, leaveTangent=1.4725, tangentMode=RCTM_User, time=0, value=0}
         {tangentMode=RCTM_User, time=1, value=1}
        ]}}}
     Scale Curve=&Multiply_Float(
      A=&MaskFloatBySpawnGroup(Masked Float=350, Passthrough Float=&RandomRangeFloat(Minimum=300, Maximum=400), Spawn Group Mask=1)
      B=$Emitter.Global Scale
     )
    )
    Index 0 Param 4=&MaskFloatBySpawnGroup(
     Masked Float=&FloatFromCurve(
      FloatCurve=DataInterface() {propertyValues {Curve {keys [
          {arriveTangent=0.1175, interpMode=RCIM_Cubic, leaveTangent=0.6205, tangentMode=RCTM_Break, time=0.22, value=-2}
          {arriveTangent=6.3488, interpMode=RCIM_Cubic, leaveTangent=6.3488, tangentMode=RCTM_User, time=0.6, value=-0.4172}
          {arriveTangent=36.7876, interpMode=RCIM_Cubic, leaveTangent=14.4841, tangentMode=RCTM_Break, time=1, value=7.3}
         ]}}}
      Scale Curve=0.2
     )
     Passthrough Float=&FloatFromCurve(
      FloatCurve=DataInterface() {propertyValues {Curve {keys [
          {arriveTangent=0.1175, interpMode=RCIM_Cubic, leaveTangent=5.607, tangentMode=RCTM_Break, time=0.1, value=-2.5}
          {arriveTangent=30.0332, interpMode=RCIM_Cubic, leaveTangent=14.4841, tangentMode=RCTM_Break, time=1, value=9}
         ]}}}
      Scale Curve=0.2
     )
     Spawn Group Mask=1
    )
    Index 1 Param 1=&FloatFromCurve(
     FloatCurve=DataInterface() {propertyValues {Curve {keys [
         {arriveTangent=0.5348, interpMode=RCIM_Cubic, leaveTangent=0.5348, tangentMode=RCTM_User, time=0.05, value=0.02}
         {arriveTangent=39.7786, interpMode=RCIM_Cubic, leaveTangent=39.7786, tangentMode=RCTM_User, time=0.4, value=8}
        ]}}}
     Scale Curve=&MaskFloatBySpawnGroup(Masked Float=85, Passthrough Float=&RandomRangeFloat(Minimum=55, Maximum=55), Spawn Group Mask=1)
    )
    Index 1 Param 2=&Divide_Float(
     A=&MaskFloatBySpawnGroup(
      Masked Float=0.01
      Passthrough Float=&RandomRangeFloat(Minimum=0.0089, Maximum=0.015)
      Spawn Group Mask=1
     )
     B=$Emitter.Global Scale
    )
    Index 1 Param 3=&FloatFromCurve(
     FloatCurve=DataInterface() {propertyValues {Curve {keys [{interpMode=RCIM_Cubic, tangentMode=RCTM_User, time=0.4, value=0.5}, <at 0.8 0.1>]}}}
     Scale Curve=5
    )
    Index 1 Param 4=&MaskFloatBySpawnGroup(Masked Float=2, Passthrough Float=1, Spawn Group Mask=1)
    Index 2 Param 1=&FloatFromCurve(FloatCurve=DataInterface() {propertyValues {Curve {keys [
         {arriveTangent=0.5431, interpMode=RCIM_Cubic, leaveTangent=0.5431, tangentMode=RCTM_User, time=0.3, value=0}
         {arriveTangent=2.1449, interpMode=RCIM_Cubic, leaveTangent=2.1449, tangentMode=RCTM_User, time=1, value=1}
        ]}}})
    Index 2 Param 2=&FloatFromCurve(
     FloatCurve=DataInterface() {propertyValues {Curve {keys [
         {interpMode=RCIM_Cubic, tangentMode=RCTM_User, time=0.3, value=0}
         {arriveTangent=2.4333, interpMode=RCIM_Cubic, leaveTangent=2.4333, tangentMode=RCTM_User, time=1, value=1}
        ]}}}
     Scale Curve=&Multiply_Float(A=0.1, B=$Emitter.Global Scale)
    )
    Index 2 Param 3=&MaskFloatBySpawnGroup(
     Masked Float=&FloatFromCurve(
      FloatCurve=DataInterface() {propertyValues {Curve {keys [
          {arriveTangent=-6.4028, interpMode=RCIM_Cubic, leaveTangent=-6.4028, tangentMode=RCTM_User, time=0.1, value=1}
          {arriveTangent=-0.0065, interpMode=RCIM_Cubic, leaveTangent=-0.0065, tangentMode=RCTM_User, time=0.4, value=0}
         ]}}}
      Scale Curve=5
     )
     Passthrough Float=&FloatFromCurve(
      FloatCurve=DataInterface() {propertyValues {Curve {keys [
          {arriveTangent=-3.4318, interpMode=RCIM_Cubic, leaveTangent=-3.4318, tangentMode=RCTM_User, time=0.1, value=1}
          {arriveTangent=-0.7527, interpMode=RCIM_Cubic, leaveTangent=-0.7527, tangentMode=RCTM_User, time=0.35, value=0}
         ]}}}
      Scale Curve=5
     )
     Spawn Group Mask=1
    )
    Index 2 Param 4=&FloatFromCurve(
     FloatCurve=DataInterface() {propertyValues {Curve {keys [
         {arriveTangent=-16.6453, interpMode=RCIM_Cubic, leaveTangent=-16.6453, tangentMode=RCTM_User, time=0, value=64}
         {arriveTangent=-87.1792, interpMode=RCIM_Cubic, leaveTangent=-87.1792, tangentMode=RCTM_User, time=0.1, value=24}
         {arriveTangent=-52.5, interpMode=RCIM_Cubic, leaveTangent=-52.5, time=0.2, value=12}
         {arriveTangent=-7.0531, interpMode=RCIM_Cubic, leaveTangent=-7.0531, tangentMode=RCTM_User, time=0.5, value=3}
         {arriveTangent=-0.6909, interpMode=RCIM_Cubic, leaveTangent=-0.6909, tangentMode=RCTM_User, time=1, value=1}
         <at 2 0.5>
         <at 3 0.3>
        ]}}}
     CurveIndex=$Emitter.Global Scale
     Scale Curve=0.6
    )
    Index 3 Param 1=&FloatFromCurve(
     FloatCurve=DataInterface() {propertyValues {Curve {keys [
         {arriveTangent=2.2353, interpMode=RCIM_Cubic, leaveTangent=2.2353, tangentMode=RCTM_User, time=0, value=0.2}
         {arriveTangent=1.4416, interpMode=RCIM_Cubic, leaveTangent=1.4416, tangentMode=RCTM_User, time=0.2, value=0.5}
         {interpMode=RCIM_Cubic, tangentMode=RCTM_User, time=0.8, value=1.25}
        ]}}}
     Scale Curve=-1
    )
    Index 3 Param 2=&FloatFromCurve(
     FloatCurve=DataInterface() {propertyValues {Curve {keys [
         {arriveTangent=5.1563, interpMode=RCIM_Cubic, leaveTangent=5.1563, tangentMode=RCTM_User, time=0, value=0}
         {arriveTangent=1.4416, interpMode=RCIM_Cubic, leaveTangent=1.4416, tangentMode=RCTM_User, time=0.1699, value=0.4}
         {interpMode=RCIM_Cubic, tangentMode=RCTM_User, time=1, value=1.1}
        ]}}}
     Scale Curve=-0.1
    )
    Index 3 Param 3=&FloatFromCurve(FloatCurve=DataInterface() {propertyValues {Curve {keys [
         {arriveTangent=-11.5239, interpMode=RCIM_Cubic, leaveTangent=-11.5239, tangentMode=RCTM_User, time=0.15, value=40}
         {arriveTangent=-88.9092, interpMode=RCIM_Cubic, leaveTangent=-88.9092, tangentMode=RCTM_User, time=0.3, value=25}
         {arriveTangent=-43.7054, interpMode=RCIM_Cubic, leaveTangent=-43.7054, tangentMode=RCTM_User, time=0.6, value=0}
         {interpMode=RCIM_Cubic, tangentMode=RCTM_User, time=0.7, value=0}
         {interpMode=RCIM_Cubic, tangentMode=RCTM_User, time=0.8, value=0}
        ]}}})
    Index 3 Param 4=&FloatFromCurve(
     FloatCurve=DataInterface() {propertyValues {Curve {keys [{tangentMode=RCTM_User, time=0, value=0}, {tangentMode=RCTM_User, time=1, value=1}]}}}
     Scale Curve=-250
    )
    Module.Index0Param0Write=true
    Module.Index0Param1Write=true
    Module.Index0Param2Write=true
    Module.Index0Param3Write=true
    Module.Index1Param0Write=true
    Module.Index1Param1Write=true
    Module.Index1Param2Write=true
    Module.Index1Param3Write=true
    Module.Index2Param0Write=true
    Module.Index2Param1Write=true
    Module.Index2Param2Write=true
    Module.Index2Param3Write=true
    Module.Index3Param0Write=true
    Module.Index3Param1Write=true
    Module.Index3Param2Write=true
    Module.Index3Param3Write=true
    Write Parameter Index 1=true
    Write Parameter Index 2=true
    Write Parameter Index 3=true
   ) {module=DynamicMaterialParameters001}
   &ScaleColor(
    Scale Alpha=&FloatFromCurve(
     FloatCurve=DataInterface() {propertyValues {Curve {keys [<at 0 1>,<at 1 0>]}}}
     Scale Curve=&FN_FrontEndFloatModifier(Gameplay=1, Front End=0.7)
    )
    Scale RGB=&MakeVectorFromLinearColorRGB(Color=&MultiplyLinearColor(
      Linear Color=&ColorFromCurve(ColorCurve=DataInterface() {propertyValues {
         AlphaCurve {keys [<at 0.0022 1>,<at 0.994 0>]}
         BlueCurve {keys [<at 0.1888 1>,<at 0.4977 0.4497>]}
         GreenCurve {keys [<at 0.1888 1>,<at 0.4977 0.4497>]}
         RedCurve {keys [<at 0.1888 1>,<at 0.4977 0.4497>]}
        }})
      Scale Factor=&MakeVector4_FromLinearColor(LinearColor=&ColorFromCurve(
        ColorCurve=DataInterface() {propertyValues {
          AlphaCurve {keys [<at 0 1>,<at 0.125 1>,<at 0.25 1>,<at 0.375 1>,<at 0.5 1>,<at 0.625 1>,<at 0.75 1>,<at 0.875 1>,<at 1 1>]}
          BlueCurve {keys [<at 0.125 0.7835>,<at 0.166 0.6172>,<at 0.208 0.5395>,<at 0.25 0.6172>,<at 0.291 0.6939>,<at 0.333 0.7379>,<at 0.375 0.7835>,<at 0.416 0.7991>,<at 0.458 0.8148>,<at 0.5 0.8308>,<at 0.541 0.8148>,<at 0.583 0.7991>,<at 0.625 0.7991>,<at 0.666 0.7529>,<at 0.708 0.7084>,<at 0.75 0.5711>,<at 0.791 0.4735>,<at 0.833 0.3231>,<at 0.8794 0.7835>]}
          GreenCurve {keys [<at 0.125 0.3118>,<at 0.166 0.5333>,<at 0.208 0.5271>,<at 0.25 0.6376>,<at 0.291 0.7529>,<at 0.333 0.7835>,<at 0.375 0.8148>,<at 0.416 0.7991>,<at 0.458 0.8148>,<at 0.5 0.8469>,<at 0.541 0.7991>,<at 0.583 0.7991>,<at 0.625 0.8148>,<at 0.666 0.7991>,<at 0.708 0.7529>,<at 0.75 0.5711>,<at 0.791 0.5029>,<at 0.833 0.5029>,<at 0.8794 0.3118>]}
          RedCurve {keys [<at 0.125 0.1252>,<at 0.166 0.7084>,<at 0.208 0.6867>,<at 0.25 0.8148>,<at 0.291 0.8879>,<at 0.333 0.8963>,<at 0.375 0.9131>,<at 0.416 0.8714>,<at 0.458 0.8714>,<at 0.5 0.9131>,<at 0.541 0.8632>,<at 0.583 0.8714>,<at 0.625 0.9131>,<at 0.666 0.8963>,<at 0.708 0.8714>,<at 0.75 0.7529>,<at 0.791 0.7529>,<at 0.833 0.8469>,<at 0.8794 0.1252>]}
         }}
        CurveIndex=Linked() {linkedVariable {name=NPC.FortniteNPC.TimeOfDay_0-1, type=NiagaraFloat}}
       ))
     ))
    ScaleA=true
    ScaleRGB=true
   )
   &SolveForcesAndVelocity(Clamp Velocity=false, Limit Acceleration=false)
   &ScaleMeshSize(Scale Factor=&VectorFromCurve(
     VectorCurve=DataInterface() {propertyValues {
       XCurve {keys [
         {arriveTangent=1.5266, interpMode=RCIM_Cubic, leaveTangent=1.5266, tangentMode=RCTM_User, time=0, value=0.5}
         {arriveTangent=1, leaveTangent=1, time=0.3, value=1.2}
         <at 1 1.5 RCIM_Cubic>
        ]}
       YCurve {keys [
         {arriveTangent=1.5266, interpMode=RCIM_Cubic, leaveTangent=1.5266, tangentMode=RCTM_User, time=0, value=0.5}
         {arriveTangent=1, leaveTangent=1, time=0.3, value=1.2}
         <at 1 1.5 RCIM_Cubic>
        ]}
       ZCurve {keys [
         {arriveTangent=1.5266, interpMode=RCIM_Cubic, leaveTangent=1.5266, tangentMode=RCTM_User, time=0, value=0.5}
         {arriveTangent=1, leaveTangent=1, time=0.3, value=1.2}
         <at 1 1.5 RCIM_Cubic>
        ]}
      }}
     Scale Curve=&Multiply_VectorByFloat(
      Vector=<1 1 1>
      Float=&FloatFromCurve(
       FloatCurve=DataInterface() {propertyValues {Curve {keys [<at 0.5 0.85 RCIM_Constant>,<at 0.6 1 RCIM_Constant>]}}}
       CurveIndex=&RandomRangeFloat(Minimum=0.5, Maximum=1)
       Scale Curve=&MaskFloatBySpawnGroup(
        Masked Float=&FloatFromCurve(
         FloatCurve=DataInterface() {propertyValues {Curve {keys [
             {arriveTangent=0.8977, interpMode=RCIM_Cubic, leaveTangent=0.8977, tangentMode=RCTM_User, time=0, value=0.6}
             {arriveTangent=0.4659, interpMode=RCIM_Cubic, leaveTangent=0.4659, tangentMode=RCTM_User, time=0.7, value=1}
            ]}}}
         Scale Curve=0.8
        )
        Passthrough Float=0.7
        Spawn Group Mask=1
       )
      )
     )
    ))
  ]
  rendererClasses [&NiagaraMeshRendererProperties]
  renderers [{rendererClass=&NiagaraMeshRendererProperties, rendererIndex=0}]
 }
 NE_Flare {
  emitterUpdateScript [
   &EmitterState(Life Cycle Mode=::NewEnumerator1, Loop Behavior=::NewEnumerator1, Loop Duration=0.5, UseLoopDelay=false)
   &SpawnBurst_Instantaneous(Spawn Count=1, Use Spawn Probability=false)
   &SpawnBurst_Instantaneous(Spawn Count=1, Spawn Group=1, Use Spawn Probability=false) {module=SpawnBurst_Instantaneous001}
   &SpawnBurst_Instantaneous(Spawn Group=2, Use Spawn Probability=false, UseLoopCountLimit=false) {module=SpawnBurst_Instantaneous002}
   &SpawnBurst_Instantaneous(Spawn Count=9, Spawn Group=3, Use Spawn Probability=false, UseLoopCountLimit=false) {module=SpawnBurst_Instantaneous003}
  ]
  particleSpawnScript [
   /SetVariables(Particles.VisibilityTag=&SpawnGroup()) {isSetParameters=true, module=SetVariables_#1}
   &InitializeParticle(
    Color=<rgba 5 0 4.1874 1>
    Lifetime=&MaskFloatBySpawnGroup(
     Masked Float=0.25
     Passthrough Float=&MaskFloatBySpawnGroup(
      Masked Float=0.4
      Passthrough Float=&MaskFloatBySpawnGroup(Masked Float=0.15, Passthrough Float=&RandomRangeFloat(Minimum=0.15, Maximum=0.25), Spawn Group Mask=2)
      Spawn Group Mask=1
     )
    )
    Mass Mode=::NewEnumerator1
    Mesh Scale=&VectorFromFloat(Value=&Multiply_Float(
      A=&MaskFloatBySpawnGroup(
       Masked Float=1
       Passthrough Float=&MaskFloatBySpawnGroup(Masked Float=0.38, Passthrough Float=0.7, Spawn Group Mask=1)
      )
      B=1
     ))
    Mesh Scale Mode=::NewEnumerator4
    Sprite Size Max=<5 40>
    Sprite Size Min=<1 20>
    Sprite Size Mode=::NewEnumerator2
    UsePositionOffset=false
    Write Material Random=true
   ) {module=InitializeParticle001}
   &InitialMeshOrientation(Orientation Axis=<0 0 1>, Orientation Vector=<0 0 1>, Rotation=<0 0.25 0>, Use Orientation Vector=true, Use Rotation Vector=true)
   &InitialMeshOrientation(
    Orientation Axis=<0 0 0>
    Orientation Vector=<0 0 0>
    Rotation=&RandomRangeVector(Minimum=<0 0 0>, Maximum=<1 0 0>)
    Use Orientation Vector=false
    Use Rotation Vector=true
   ) {module=InitialMeshOrientation001}
   /SetVariables(Particles.BigFlareOrient:Quat4f=&OrientMeshToVector(Look At Direction=<0 0 0>, Reference Axis Vector=<1 0 0>)) {isSetParameters=true}
   /SetVariables(Particles.BigFlareDynmat:Vector4f=&MakeVector4(X=&UniformRangedFloat(Minimum=0, Maximum=20), Y=0, Z=0, W=0)) {isSetParameters=true}
   &AddVelocity(
    Module.Use Falloff Distance=false
    Override Seed=false
    Velocity Mode=::NewEnumerator1
    Velocity Speed=&MaskFloatBySpawnGroup(
     Masked Float=0
     Passthrough Float=&MaskFloatBySpawnGroup(
      Masked Float=0
      Passthrough Float=&MaskFloatBySpawnGroup(Masked Float=0, Passthrough Float=&RandomRangeFloat(Minimum=100, Maximum=200), Spawn Group Mask=2)
      Spawn Group Mask=1
     )
    )
   )
  ]
  particleUpdateScript [
   &ParticleState()
   &ScaleSpriteSize(
    Non-Uniform Curve Sprite Scale=DataInterface() {propertyValues {XCurve {keys [<at 0 0.75 RCIM_Cubic>,<at 1 1 RCIM_Cubic>]}, YCurve {keys [<at 0 1 RCIM_Cubic>,<at 1 2 RCIM_Cubic>]}}}
    Scale Sprite Size Mode=::NewEnumerator3
   )
   &ScaleMeshSize(Scale Factor=&VectorFromFloat(Value=&MaskFloatBySpawnGroup(
      Masked Float=&FloatFromCurve(FloatCurve=DataInterface() {propertyValues {Curve {keys [{arriveTangent=5.2157, interpMode=RCIM_Cubic, leaveTangent=5.2157, tangentMode=RCTM_User, time=0, value=0}, <at 0.4 1 RCIM_Cubic>]}}})
      Passthrough Float=&MaskFloatBySpawnGroup(
       Masked Float=&FloatFromCurve(FloatCurve=DataInterface() {propertyValues {Curve {keys [
            {interpMode=RCIM_Cubic, tangentMode=RCTM_Break, time=0, value=0.2}
            {arriveTangent=2.9365, interpMode=RCIM_Cubic, leaveTangent=2.9365, tangentMode=RCTM_User, time=0.1, value=1}
            {interpMode=RCIM_Cubic, tangentMode=RCTM_Break, time=1, value=2}
           ]}}})
       Passthrough Float=&FloatFromCurve(FloatCurve=DataInterface() {propertyValues {Curve {keys [
            {interpMode=RCIM_Cubic, leaveTangent=2.2298, tangentMode=RCTM_Break, time=0, value=0}
            {arriveTangent=1.1172, interpMode=RCIM_Cubic, leaveTangent=0.7073, tangentMode=RCTM_Break, time=0.4963, value=0.8}
            {interpMode=RCIM_Cubic, tangentMode=RCTM_Break, time=1, value=1}
           ]}}})
       Spawn Group Mask=1
      )
     )))
   &ScaleColor(
    Scale Alpha=&MaskFloatBySpawnGroup(
     Masked Float=0
     Passthrough Float=&FloatFromCurve(FloatCurve=DataInterface() {propertyValues {Curve {keys [
          {interpMode=RCIM_Cubic, tangentMode=RCTM_User, time=0, value=1}
          {arriveTangent=-2.6886, interpMode=RCIM_Cubic, leaveTangent=-2.6886, tangentMode=RCTM_User, time=1, value=0}
         ]}}})
    )
    Scale RGB=&MaskVectorBySpawnGroup(
     Masked Vector=&VectorFromFloat(Value=&FloatFromCurve(FloatCurve=DataInterface() {propertyValues {Curve {keys [<at 0 0.7>,<at 0.7729 0.7>,<at 1 0>]}}}, Scale Curve=0.2))
     Passthrough Vector=&Multiply_Vector(
      A=&MakeVectorFromLinearColorRGB(Color=&ColorFromCurve(ColorCurve=DataInterface() {propertyValues {
          AlphaCurve {keys [<at 0 1>,<at 0.4696 1>,<at 1 0>]}
          BlueCurve {keys [<at 0 0>,<at 0.2681 0>,<at 0.4696 0>,<at 1 0>]}
          GreenCurve {keys [<at 0 0.5481>,<at 0.2681 0.3605>,<at 0.4696 0.1134>,<at 1 0.049>]}
          RedCurve {keys [<at 0 1>,<at 0.2681 1>,<at 0.4696 0.6979>,<at 1 0.3698>]}
         }}))
      B=&VectorFromFloat(Value=&FloatFromCurve(FloatCurve=DataInterface() {propertyValues {Curve {keys [
            {interpMode=RCIM_Cubic, tangentMode=RCTM_Break, time=0, value=10}
            {arriveTangent=-6.6072, interpMode=RCIM_Cubic, leaveTangent=-6.6072, tangentMode=RCTM_User, time=0.15, value=3}
            {interpMode=RCIM_Cubic, tangentMode=RCTM_Break, time=1, value=1}
           ]}}}))
     )
    )
    ScaleA=true
    ScaleRGB=true
   )
   &DynamicMaterialParameters(
    Index 0 Param 1=&FloatFromCurve(FloatCurve=DataInterface() {propertyValues {Curve {keys [{arriveTangent=-2.1015, interpMode=RCIM_Cubic, leaveTangent=-2.1015, tangentMode=RCTM_User, time=0, value=0.5}, <at 1 -0.15>]}}})
    Index 0 Param 2=&FloatFromCurve(FloatCurve=DataInterface() {propertyValues {Curve {keys [{arriveTangent=-0.3464, interpMode=RCIM_Cubic, leaveTangent=-0.3464, tangentMode=RCTM_User, time=0, value=1}, <at 1 0.9>]}}})
    Index 0 Param 3=&FloatFromCurve(FloatCurve=DataInterface() {propertyValues {Curve {keys [
         {interpMode=RCIM_Cubic, tangentMode=RCTM_Break, time=0, value=0}
         {arriveTangent=0.4478, interpMode=RCIM_Cubic, leaveTangent=0.4478, tangentMode=RCTM_User, time=0.15, value=1}
         {interpMode=RCIM_Cubic, tangentMode=RCTM_Break, time=1, value=1.2}
        ]}}})
    Module.Param0WriteEnabled=true
    Module.Param1WriteEnabled=true
    Module.Param2WriteEnabled=true
    Module.Param3WriteEnabled=false
   )
   /SetVariables(Particles.BigFlareScale:Vector3f=&Multiply_VectorByFloat(Vector=$Particles.Scale, Float=0.7)) {isSetParameters=true}
   /SetVariables(Particles.BigFlareColor:LinearColor=&MultiplyLinearColor(
     Linear Color=<rgba 5 0 4.1874 1>
     Scale Factor=&MakeFloat4FromVectorAndFloat(
      "Vector (RGB)"=&VectorFromFloat(Value=1)
      "Float (Alpha)"=&FloatFromCurve(FloatCurve=DataInterface() {propertyValues {Curve {keys [<at 0.4 1 RCIM_Cubic>,<at 1 0>]}}})
     )
    )) {isSetParameters=true}
   /SetVariables(
    Particles.OrangeGlow_Dynamic:Vector4f=&MakeVector4(
     X=&FloatFromCurve(FloatCurve=DataInterface() {propertyValues {Curve {keys [<at 0 1>,<at 0.3 0.85>,<at 1 1>]}}})
     Y=0
     Z=0
     W=0
    )
    Particles.Orangr_Glow:LinearColor=&ColorFromCurve(ColorCurve=DataInterface() {propertyValues {
       AlphaCurve {keys [<at 0.0016 1 RCIM_Cubic>,<at 0.7 0.5387>,<at 0.9722 0 RCIM_Cubic>]}
       BlueCurve {keys [<at 0 0 RCIM_Cubic>,<at 1 0 RCIM_Cubic>]}
       GreenCurve {keys [<at 0 1.3734 RCIM_Cubic>,<at 1 0 RCIM_Cubic>]}
       RedCurve {keys [<at 0 5 RCIM_Cubic>,<at 1 0 RCIM_Cubic>]}
      }})
   ) {isSetParameters=true}
   /SetVariables(Particles.Streak_Color:LinearColor=&ColorFromCurve(ColorCurve=DataInterface() {propertyValues {
       AlphaCurve {keys [<at 0 1 RCIM_Cubic>,<at 1 0 RCIM_Cubic>]}
       BlueCurve {keys [<at 0 0 RCIM_Cubic>,<at 1 1 RCIM_Cubic>]}
       GreenCurve {keys [<at 0 37.7 RCIM_Cubic>,<at 1 1 RCIM_Cubic>]}
       RedCurve {keys [<at 0 100 RCIM_Cubic>,<at 1 1 RCIM_Cubic>]}
      }})) {isSetParameters=true}
   &SolveForcesAndVelocity(Clamp Velocity=false, Limit Acceleration=false)
  ]
  rendererClasses [&NiagaraMeshRendererProperties, &NiagaraSpriteRendererProperties]
  renderers [
   {rendererClass=&NiagaraMeshRendererProperties, rendererIndex=0}
   {rendererClass=&NiagaraMeshRendererProperties, rendererIndex=1}
   {rendererClass=&NiagaraMeshRendererProperties, rendererIndex=2}
   {rendererClass=&NiagaraSpriteRendererProperties, rendererIndex=3}
  ]
 }
 Souce {
  emitterUpdateScript [
   &EmitterState(Inactive Response=::NewEnumerator1, Life Cycle Mode=::NewEnumerator1, Loop Behavior=::NewEnumerator1, Loop Duration=2, UseLoopDelay=false)
   &SpawnBurst_Instantaneous(Spawn Count=15, Use Spawn Probability=false, UseLoopCountLimit=false)
  ]
  particleSpawnScript [
   &InitializeParticle(
    Lifetime Max=0.5
    Lifetime Min=0.1
    Lifetime Mode=::NewEnumerator1
    Mesh Scale Mode=::NewEnumerator1
    Mesh Uniform Scale Max=1
    Mesh Uniform Scale Min=0.25
    Sprite Rotation Mode=::NewEnumerator2
    Sprite Size Max=<5 20>
    Sprite Size Min=<5 20>
    Sprite Size Mode=::NewEnumerator2
    Sprite UV Mode=::NewEnumerator3
    UsePositionOffset=false
    Write Material Random=true
   )
   &ShapeLocation(MaskSpawnGroup=false, Override Seed=false, Sphere Radius=20, Sphere Surface Distribution=1, UseOwnerScale=true)
   &AddVelocity(
    Cone Axis=<0 90 0>
    Module.UseConeFalloff=false
    Module.UseVelDistribution=false
    Override Seed=false
    Velocity Mode=::NewEnumerator2
    Velocity Speed=&RandomRangeFloat(Minimum=150, Maximum=750)
   ) {module=AddVelocity001}
   &AddVelocity(Velocity=&RandomRangeVector(Minimum=<0 0 250>, Maximum=<0 0 0>)) {module=AddVelocity002}
   &InitialMeshOrientation(Use Rotation Vector=false)
  ]
  particleUpdateScript [
   &Drag(Drag=2, Use Linear Drag=true, Use Rotational Drag=true)
   &CurlNoiseForce(
    Bias Noise=false
    Noise Strength=&Multiply_Float(
     A=1000
     B=&FloatFromCurve(FloatCurve=DataInterface() {propertyValues {Curve {keys [<at 0 0 RCIM_Cubic>,<at 0.5 1 RCIM_Cubic>]}}}, Scale Curve=1.3)
    )
    Pan Noise=false
   )
   &GravityForce(Gravity=&RandomRangeVector(Minimum=<0 0 -600>, Maximum=<0 0 -700>))
   &SolveForcesAndVelocity(Clamp Velocity=false, Limit Acceleration=false)
   &ParticleState()
   /SetVariables(Particles.Position=$Particles.Position) {isSetParameters=true}
   &ScaleColor(
    Scale Alpha=&FloatFromCurve(FloatCurve=DataInterface() {propertyValues {Curve {keys [<at 1 1>,<at 1.3 0>,<at 1.8 0>]}}}, CurveIndex=$Emitter.Age)
    Scale RGB=&Multiply_VectorByFloat(
     Vector=&MakeVectorFromLinearColorRGB(Color=&ColorFromCurve(ColorCurve=DataInterface() {propertyValues {
         AlphaCurve {keys []}
         BlueCurve {keys [<at 0.5026 2.5124 RCIM_Cubic>]}
         GreenCurve {keys [<at 0.5026 0 RCIM_Cubic>]}
         RedCurve {keys [<at 0.5026 3 RCIM_Cubic>]}
        }}))
     Float=0.3
    )
    ScaleA=true
    ScaleRGB=true
   )
   &FN_Sprite_Attributes(
    Color=&MakeLinearColorFromVectorAndFloat(
     "Vector (RGB)"=&MakeVectorFromLinearColorRGB(Color=<rgba 1 0.2297 0 0.2>)
     "Float (Alpha)"=&Multiply_Float(
      A=0.2
      B=&FloatFromCurve(
       FloatCurve=DataInterface() {propertyValues {Curve {keys [<at 1 1>,<at 1.5 0>,<at 1.8 0>]}}}
       CurveIndex=$Emitter.Age
       Scale Curve=&FN_FrontEndFloatModifier(Gameplay=1, Front End=0.5)
      )
     )
    )
    Create Color Attribute=true
    Create SpriteSize Attribute=true
    SpriteSize=&Multiply_Vector2D_ByFloat(Float=20, Vector2D=<1 2>)
    Write to Module Namespace=true
   )
   &FN_Light_Attributes(
    Color=<rgba 15 3.75 0 1>
    Create Color Attribute=true
    Create Exponent Attribute=true
    Create Radius Attribute=true
    Enable Particle Light Enable=true
    Exponent=5
    Radius=70
    Scale Color=&FloatFromCurve(
     FloatCurve=DataInterface() {propertyValues {Curve {keys [<at 0 1>,<at 1.4 1>,<at 1.7 0>,<at 2 0>]}}}
     CurveIndex=$Emitter.Age
     Scale Curve=&FN_FrontEndFloatModifier(Gameplay=1, Front End=0)
    )
   )
  ]
  rendererClasses [&NiagaraSpriteRendererProperties]
  renderers [
   {rendererClass=&NiagaraSpriteRendererProperties}
   {rendererClass=&NiagaraSpriteRendererProperties, rendererIndex=1}
  ]
 }
 trail {
  emitterUpdateScript [
   &EmitterState(Inactive Response=::NewEnumerator1, Life Cycle Mode=::NewEnumerator1, Loop Behavior=::NewEnumerator1, Loop Duration=2, UseLoopDelay=false)
   &SpawnParticlesFromOtherEmitter(
    Attribute Reader=DataInterface() {propertyValues {EmitterBinding {bindingMode=Other, emitterName=Souce}}}
    Spawn Rate=15
    Use Maximum Age=false
    Use Minimum Age=false
   )
  ]
  particleSpawnScript [
   &InitializeParticle(
    Color Mode=::NewEnumerator0
    Lifetime=0.3
    Ribbon Width=3
    Ribbon Width Mode=::NewEnumerator1
    UsePositionOffset=false
    Write Material Random=true
   )
   &SampleParticlesFromOtherEmitter(
    Module.MaxSampledAgeCutoff=false
    Module.MinSampledAgeCutoff=false
    Particle ID Sampling=::NewEnumerator1
    Reject Particle By Sampled Age=true
    Transform Data Option=1
   )
   /SetVariables(Particles.Random=&RandomRangeFloat(Minimum=0.7, Maximum=1.1)) {isSetParameters=true}
  ]
  particleUpdateScript [
   &SolveForcesAndVelocity(Clamp Velocity=false, Limit Acceleration=false)
   &ParticleState()
   &ScaleRibbonWidth(Ribbon Width Scale=&FloatFromCurve(FloatCurve=DataInterface() {propertyValues {Curve {keys [{arriveTangent=-1.1377, interpMode=RCIM_Cubic, leaveTangent=-1.1377, tangentMode=RCTM_User, time=0, value=1}, <at 1 0.3>]}}}))
   &FN_Mesh_Attributes(
    Color=&MakeLinearColorFromVectorAndFloat(
     "Vector (RGB)"=&Multiply_Vector(
      A=&MakeVectorFromLinearColorRGB(Color=&ColorFromCurve(
        ColorCurve=DataInterface() {propertyValues {
          AlphaCurve {keys []}
          BlueCurve {keys [<at 0.0278 0.8375 RCIM_Cubic>, {interpMode=RCIM_Cubic, leaveTangent=0.8203, time=0.6108, value=0.2865}]}
          GreenCurve {keys [<at 0.0278 0 RCIM_Cubic>, {interpMode=RCIM_Cubic, leaveTangent=0.5619, time=0.6108, value=0.2865}]}
          RedCurve {keys [<at 0.0278 1 RCIM_Cubic>, {interpMode=RCIM_Cubic, leaveTangent=-1.3942, time=0.6108, value=0.2865}]}
         }}
        CurveIndex=$Emitter.Age
       ))
      B=<1 1 1>
     )
     "Float (Alpha)"=&Multiply_Float(
      A=&FloatFromCurve(
       FloatCurve=DataInterface() {propertyValues {Curve {keys [
           {interpMode=RCIM_Cubic, tangentMode=RCTM_User, time=0, value=1}
           <at 0.1 1 RCIM_Cubic>
           {interpMode=RCIM_Cubic, tangentMode=RCTM_User, time=0.3, value=1}
           {arriveTangent=-1.4286, interpMode=RCIM_Cubic, leaveTangent=-1.4286, time=0.6, value=1}
           {tangentMode=RCTM_User, time=1, value=0}
          ]}}}
       Scale Curve=&FloatFromCurve(FloatCurve=DataInterface() {propertyValues {Curve {keys [<at 0 1>,<at 1.7 1>,<at 2 0>]}}}, CurveIndex=$Emitter.Age)
      )
      B=1
     )
    )
    Create Color Attribute=true
    Write to Module Namespace=true
   ) {module=FN_Mesh_Attributes001}
   &DynamicMaterialParameters(
    Index 0 Param 1=&Multiply_Float(A=0, B=$Particles.Random)
    Index 0 Param 2=&FloatFromCurve(FloatCurve=DataInterface() {propertyValues {Curve {keys [{tangentMode=RCTM_User, time=0, value=1}, {tangentMode=RCTM_User, time=0.5, value=1}]}}})
    Index 0 Param 3=&FloatFromCurve(FloatCurve=DataInterface() {propertyValues {Curve {keys [<at 0.4 0>,<at 1 1>]}}})
    Index 1 Param 1=&Multiply_Float(A=0.2, B=$Particles.Random)
    Index 1 Param 2=&FloatFromCurve(FloatCurve=DataInterface() {propertyValues {Curve {keys [{tangentMode=RCTM_User, time=0, value=1}, {tangentMode=RCTM_User, time=1, value=1}]}}})
    Index 1 Param 3=&FloatFromCurve(FloatCurve=DataInterface() {propertyValues {Curve {keys [<at 0.7 0>,<at 1 0>]}}})
    Index 2 Param 2=0.8
    Module.Index0Float0Write=true
    Module.Index0Float1Write=true
    Module.Index0Float2Write=true
    Module.Index0Float3Write=true
    Module.Index1Float0Write=true
    Module.Index1Float1Write=true
    Module.Index1Float2Write=true
    Module.Index1Float3Write=true
    Module.Index2Float0Write=true
    Module.Index2Float1Write=true
    Module.Index2Float2Write=true
    Module.Index2Float3Write=true
    Write Parameter Index 1=::NewEnumerator1
    Write Parameter Index 2=::NewEnumerator1
   )
  ]
  rendererClasses [&NiagaraRibbonRendererProperties]
  renderers [
   {rendererClass=&NiagaraRibbonRendererProperties}
   {rendererClass=&NiagaraRibbonRendererProperties, rendererIndex=1}
  ]
 }
}
summary {
 systemName=NS_Emote_CyberMitt_Explosion
 emitters [
  {emitterName=NE_Flare, rendererClasses [&NiagaraMeshRendererProperties, &NiagaraSpriteRendererProperties]}
  {emitterName=Main_Smoke, rendererClasses [&NiagaraMeshRendererProperties]}
  {emitterName=trail, rendererClasses [&NiagaraRibbonRendererProperties]}
  {emitterName=Souce, rendererClasses [&NiagaraSpriteRendererProperties]}
  {
   emitterName=Debris_Spark_Explosion
   rendererClasses [&NiagaraMeshRendererProperties, &NiagaraSpriteRendererProperties]
  }
 ]
 userVariables [
  User.bIsFrontEnd {type=NiagaraBool}
  User.bIsFrontEndPreview {type=NiagaraBool}
 ]
}
