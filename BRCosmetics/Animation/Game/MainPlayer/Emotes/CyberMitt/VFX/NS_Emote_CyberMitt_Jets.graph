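# Niagara jet exhaust effect for the CyberMitt emote, pairing skeletal-mesh anim trails with a thruster mesh.
# Two ribbon emitters (FillTrail, blue-tinted FillTrail005) follow bones through shared AnimTrail functions, with lifetime scaled by User.BatTrail, which defaults to zero.
# Thruster continuously spawns oriented meshes that stretch in Z when User.OwnerLocalVelocity passes a threshold.
# Flame_Burst, driven by User.Burst, is disabled, and Glow has no renderers.
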NiagaraSystem /BRCosmetics/Animation/Game/MainPlayer/Emotes/CyberMitt/VFX/NS_Emote_CyberMitt_Jets
dependencies {
 usedDataInterfaces [
  &NiagaraDataInterfaceColorCurve
  &NiagaraDataInterfaceCurve
  &NiagaraDataInterfaceVectorCurve
  &NiagaraDataInterfaceArrayFloat
 ]
 usedDynamicInputs [&SimulationPosition, &MakeCustomVectorFromBool]
 usedModules [
  &self:SystemSpawnScript.SystemScriptSource.SystemScriptGraph.NiagaraNodeAssignment_0.SetVariables_#1
  &self:SystemSpawnScript.SystemScriptSource.SystemScriptGraph.NiagaraNodeAssignment_1.SetVariables_#2
  &SystemState
  &self:Set_Current_Exhaust
  &self:Set_Trail_Length
  &self:Set_Trail_Opacity
  &Timeline
 ]
}
emitters {
 FillTrail {
  emitterSpawnScript [&FN_RibbonIDOffsetUtility()]
  emitterUpdateScript [
   &EmitterState(Life Cycle Mode=::NewEnumerator1, Loop Behavior=::NewEnumerator1, Loop Duration Mode=::NewEnumerator1, UseLoopDelay=false)
   &FN_SpawnAnimTrail(Switch on ENiagara_AnimTrail_Source:ENiagara_AnimTrail_Source=::NewEnumerator1)
  ]
  particleSpawnScript [
   &FN_InitializeAnimTrailBoneIndecies() {module=FN_InitializeAnimTrailBoneIndecies001}
   &FN_InitializeAnimTrail(Lifetime=&Multiply_Float(A=0.1, B=$User.BatTrail), SkeletalMesh=$System.Skeletal Mesh)
  ]
  particleUpdateScript [
   &FN_UpdateAnimTrail(SkeletalMesh=$System.Skeletal Mesh) {module=FN_UpdateAnimTrail001}
   /SetVariables(Particles.RibbonWidth_02:NiagaraFloat=&Multiply_Float(A=$Particles.RibbonWidth, B=0.25)) {isSetParameters=true}
   &ScaleColor(
    Scale Alpha=&Multiply_Float(
     A=1
     B=&FloatFromCurve(
      FloatCurve=DataInterface() {propertyValues {Curve {
         defaultValue=1
         keys [{interpMode=RCIM_Cubic, tangentMode=RCTM_User, time=0.1, value=1}, {interpMode=RCIM_Cubic, tangentMode=RCTM_User, time=0.18, value=0}]
        }}}
      CurveIndex=$System.Age
     )
    )
    Scale RGB=&MakeVectorFromLinearColorRGB(Color=&ColorFromCurve(ColorCurve=DataInterface() {propertyValues {
        AlphaCurve {keys [<at 0 1 RCIM_Cubic>,<at 1 1 RCIM_Cubic>]}
        BlueCurve {keys [<at 0.4811 0.0169 RCIM_Cubic>,<at 0.8995 1 RCIM_Cubic>]}
        GreenCurve {keys [<at 0.4811 0.0173 RCIM_Cubic>,<at 0.8995 1 RCIM_Cubic>]}
        RedCurve {keys [<at 0.4811 0.02 RCIM_Cubic>,<at 0.8995 1 RCIM_Cubic>]}
       }}))
    ScaleA=false
    ScaleRGB=true
   ) {module=ScaleColor001}
   &ScaleColor(Scale Alpha=0.5, ScaleA=true, ScaleRGB=true) {module=ScaleColor003}
   &ScaleColor(
    Scale Alpha=&FloatFromCurve(FloatCurve=DataInterface() {propertyValues {Curve {keys [{interpMode=RCIM_Cubic, tangentMode=RCTM_Break, time=0, value=1}, {interpMode=RCIM_Cubic, tangentMode=RCTM_Break, time=1, value=0}]}}})
    ScaleA=true
    ScaleRGB=true
   ) {module=ScaleColor004}
  ]
  rendererClasses [&NiagaraRibbonRendererProperties]
  renderers [
   {rendererClass=&NiagaraRibbonRendererProperties}
   {rendererClass=&NiagaraRibbonRendererProperties, rendererIndex=1}
  ]
 }
 FillTrail005 {
  emitterSpawnScript [&FN_RibbonIDOffsetUtility()]
  emitterUpdateScript [
   &EmitterState(Life Cycle Mode=::NewEnumerator1, Loop Behavior=::NewEnumerator1, Loop Delay=2.9, Loop Duration Mode=::NewEnumerator1, UseLoopDelay=false)
   &FN_SpawnAnimTrail(AnimTrailWidthMode=::FromSecond, Switch on ENiagara_AnimTrail_Source:ENiagara_AnimTrail_Source=::NewEnumerator1)
  ]
  particleSpawnScript [
   &FN_InitializeAnimTrailBoneIndecies() {module=FN_InitializeAnimTrailBoneIndecies001}
   &FN_InitializeAnimTrail(
    AnimTrailWidth=0.5
    Color=<rgba 0 0.2117 1 1>
    Lifetime=&Multiply_Float(A=0.07, B=$User.BatTrail)
    SkeletalMesh=$System.Skeletal Mesh
   )
  ]
  particleUpdateScript [
   &FN_UpdateAnimTrail(SkeletalMesh=$System.Skeletal Mesh) {module=FN_UpdateAnimTrail001}
   &ScaleColor(
    Scale Alpha=&FloatFromCurve(FloatCurve=DataInterface() {propertyValues {Curve {
        defaultValue=0
        keys [{interpMode=RCIM_Cubic, tangentMode=RCTM_User, time=0, value=2}, {interpMode=RCIM_Cubic, tangentMode=RCTM_User, time=1, value=1}]
       }}})
    ScaleA=true
    ScaleRGB=true
   ) {module=ScaleColor001}
   &ScaleColor(Scale Alpha=0.5, ScaleA=true, ScaleRGB=true) {module=ScaleColor003}
  ]
  rendererClasses [&NiagaraRibbonRendererProperties]
  renderers [{rendererClass=&NiagaraRibbonRendererProperties}]
 }
 Flame_Burst {
  bEnabled=false
  emitterUpdateScript [
   &EmitterState(
    Life Cycle Mode=::NewEnumerator1
    Loop Behavior=::NewEnumerator1
    Loop Duration Mode=::NewEnumerator1
    Scalability Mode=::NewEnumerator1
    UseLoopDelay=false
   )
   &SpawnRate(SpawnRate=$User.Burst, Use Spawn Probability=false)
  ]
  particleSpawnScript [
   &InitializeParticle(
    Lifetime=0.3
    Mesh Scale=<0.075 0.035 2.5>
    Mesh Scale Mode=::NewEnumerator4
    Position Offset=<0 34 -5>
    UsePositionOffset=true
    Write Material Random=false
   ) {module=InitializeParticle001}
   &DynamicMaterialParameters(
    Index 0 Param 1=5
    Index 0 Param 2=5
    Module.Index0Float0Write=true
    Module.Index0Float1Write=true
    Module.Index0Float2Write=true
    Module.Index0Float3Write=true
   )
   &InitialMeshOrientation(
    Mesh Axis To Orient=<0 1 0>
    Mesh Orientation Mode=::NewEnumerator2
    Orientation Vector=<0 0 -1>
    Rotation=<0 0 0.5>
    Use Rotation Vector=true
   )
  ]
  particleUpdateScript [
   &ParticleState()
   &ScaleMeshSize(Scale Factor=&VectorFromCurve(VectorCurve=DataInterface() {propertyValues {
       XCurve {keys [<at 0 1>,<at 0.5 1>,<at 1 1>]}
       YCurve {keys [<at 0 1>,<at 0.5 1>,<at 1 1>]}
       ZCurve {keys [<at 0 1 RCIM_Cubic>,<at 0.25 3 RCIM_Cubic>,<at 1 1>]}
      }})) {module=ScaleMeshSize001}
  ]
  rendererClasses [&NiagaraMeshRendererProperties]
  renderers [{rendererClass=&NiagaraMeshRendererProperties, rendererIndex=0}]
 }
 Glow {rendererClasses [], renderers []}
 Thruster {
  emitterUpdateScript [
   &EmitterState()
   &SpawnRate(SpawnRate=5, Use Spawn Probability=false)
  ]
  particleSpawnScript [
   &InitializeParticle(
    Color=<rgba 8 8 8 1>
    Lifetime=0.8
    Mesh Scale=<0.075 0.035 2.5>
    Mesh Scale Mode=::NewEnumerator4
    Position Offset=<0 34 -5>
    UsePositionOffset=true
    Write Material Random=true
   )
   &InitialMeshOrientation(Mesh Orientation Mode=::NewEnumerator7, Rotation=<-0.25 0 0>, Use Rotation Vector=true)
   &DynamicMaterialParameters(
    Index 0 Param 1=5
    Index 0 Param 2=5
    Module.Index0Float0Write=true
    Module.Index0Float1Write=true
    Module.Index0Float2Write=true
    Module.Index0Float3Write=true
   )
  ]
  particleUpdateScript [
   &ParticleState()
   &ScaleMeshSize(Scale Factor=&MakeVector(
     X=1
     Y=1
     Z=&MakeCustomFloatFromBool(
      Bool=&SetBoolByFloatComparison(A=&VectorLength(Vector=$User.OwnerLocalVelocity), B=10)
      True Float=1.5
      False Float=1
     )
    ))
  ]
  rendererClasses [&NiagaraMeshRendererProperties]
  renderers [{rendererClass=&NiagaraMeshRendererProperties, rendererIndex=0}]
 }
}
summary {
 systemName=NS_Emote_CyberMitt_Jets
 emitters [
  {emitterName=FillTrail, rendererClasses [&NiagaraRibbonRendererProperties]}
  {emitterName=FillTrail005, rendererClasses [&NiagaraRibbonRendererProperties]}
  {bEnabled=false, emitterName=Flame_Burst, rendererClasses [&NiagaraMeshRendererProperties]}
  {emitterName=Glow, rendererClasses []}
  {emitterName=Thruster, rendererClasses [&NiagaraMeshRendererProperties]}
 ]
 userVariables [
  User.BatTrail {type=NiagaraFloat, defaultValue=0, description=""}
  User.bIsFrontEnd {type=NiagaraBool}
  User.bIsFrontEndPreview {type=NiagaraBool}
  User.Burst {type=NiagaraFloat, defaultValue=1, description=""}
  User.ForwardDot {type=NiagaraFloat, defaultValue=0, description=""}
  User.OwnerLocalVelocity {type=Vector3f, description="", defaultValue {struct=Vector3f, value=<0 0 0>}}
  User.PlayerSpeed {type=NiagaraFloat, defaultValue=0, description=""}
  User.RibbonIDOffset {type=NiagaraInt32}
  User.RightDot {type=NiagaraFloat, defaultValue=0, description=""}
  User.Skeletal Mesh {
   type=NiagaraDataInterfaceSkeletalMesh
   defaultValue {
    struct=NiagaraExt_VariableValue_DataInterface
    value {dataInterface=&self:NiagaraDataInterfaceSkeletalMesh_2}
   }
  }
  User.SkeletalMesh {
   type=NiagaraDataInterfaceSkeletalMesh
   defaultValue {
    struct=NiagaraExt_VariableValue_DataInterface
    value {dataInterface=&self:NiagaraDataInterfaceSkeletalMesh_0}
   }
  }
 ]
}
