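# Niagara system for the CyberMitt emote, named as a jet burst, whose only emitter, Flame_Mesh, is disabled.
# That emitter burst-spawns a short-lived, elongated mesh particle with a position offset, a downward orientation and dynamic material parameters.
# A vector curve stretches the mesh along Z early in its life, then eases it back.
# User parameters for player speed, direction, front-end flags and skeletal meshes are declared but unreferenced in the listed modules.
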NiagaraSystem /BRCosmetics/Animation/Game/MainPlayer/Emotes/CyberMitt/VFX/NS_Emote_CyberMitt_Jets_Burst
dependencies {
 usedDataInterfaces [&NiagaraDataInterfaceVectorCurve]
 usedDynamicInputs [&SimulationPosition]
 usedModules [&SystemState]
}
emitters {Flame_Mesh {
  bEnabled=false
  emitterUpdateScript [
   &EmitterState(
    Life Cycle Mode=::NewEnumerator1
    Loop Behavior=::NewEnumerator1
    Loop Delay=0.2
    Loop Duration=0.1
    Scalability Mode=::NewEnumerator1
    UseLoopDelay=false
   )
   &SpawnBurst_Instantaneous(Use Spawn Probability=false, UseLoopCountLimit=false)
  ]
  particleSpawnScript [
   &InitializeParticle(
    Lifetime=0.3
    Mesh Scale=<0.075 0.035 2.5>
    Mesh Scale Mode=::NewEnumerator4
    Position Offset=<0 34 -5>
    UsePositionOffset=true
    Write Material Random=false
   ) {module=InitializeParticle001}
   &DynamicMaterialParameters(
    Index 0 Param 1=5
    Index 0 Param 2=5
    Module.Index0Float0Write=true
    Module.Index0Float1Write=true
    Module.Index0Float2Write=true
    Module.Index0Float3Write=true
   )
   &InitialMeshOrientation(
    Mesh Axis To Orient=<0 1 0>
    Mesh Orientation Mode=::NewEnumerator2
    Orientation Vector=<0 0 -1>
    Rotation=<0 0 0.5>
    Use Rotation Vector=true
   )
  ]
  particleUpdateScript [
   &ParticleState()
   &ScaleMeshSize(Scale Factor=&VectorFromCurve(VectorCurve=DataInterface() {propertyValues {
       XCurve {keys [<at 0 1>,<at 0.5 1>,<at 1 1>]}
       YCurve {keys [<at 0 1>,<at 0.5 1>,<at 1 1>]}
       ZCurve {keys [<at 0 1 RCIM_Cubic>,<at 0.25 3 RCIM_Cubic>,<at 1 1>]}
      }})) {module=ScaleMeshSize001}
  ]
  rendererClasses [&NiagaraMeshRendererProperties]
  renderers [{rendererClass=&NiagaraMeshRendererProperties, rendererIndex=0}]
 }}
summary {
 systemName=NS_Emote_CyberMitt_Jets_Burst
 emitters [{bEnabled=false, emitterName=Flame_Mesh, rendererClasses [&NiagaraMeshRendererProperties]}]
 userVariables [
  User.bIsFrontEnd {type=NiagaraBool}
  User.bIsFrontEndPreview {type=NiagaraBool}
  User.ForwardDot {type=NiagaraFloat, defaultValue=0, description=""}
  User.PlayerSpeed {type=NiagaraFloat, defaultValue=0, description=""}
  User.RightDot {type=NiagaraFloat, defaultValue=0, description=""}
  User.Skeletal Mesh {
   type=NiagaraDataInterfaceSkeletalMesh
   defaultValue {
    struct=NiagaraExt_VariableValue_DataInterface
    value {dataInterface=&self:NiagaraDataInterfaceSkeletalMesh_2}
   }
  }
  User.SkeletalMesh {
   type=NiagaraDataInterfaceSkeletalMesh
   defaultValue {
    struct=NiagaraExt_VariableValue_DataInterface
    value {dataInterface=&self:NiagaraDataInterfaceSkeletalMesh_0}
   }
  }
 ]
}
