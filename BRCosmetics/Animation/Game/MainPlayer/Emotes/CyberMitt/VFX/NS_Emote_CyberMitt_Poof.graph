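# Niagara system for the poof effect of the CyberMitt main player emote cosmetic.
# It holds two emitters, Smoke and Spark, and lists SystemState as its only used module.
# No renderers, data interfaces, dynamic inputs or user variables are listed.
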NiagaraSystem /BRCosmetics/Animation/Game/MainPlayer/Emotes/CyberMitt/VFX/NS_Emote_CyberMitt_Poof
dependencies {usedDataInterfaces [], usedDynamicInputs [], usedModules [&SystemState]}
emitters {Smoke {rendererClasses [], renderers []}, Spark {rendererClasses [], renderers []}}
summary {
 systemName=NS_Emote_CyberMitt_Poof
 emitters [{emitterName=Smoke, rendererClasses []}, {emitterName=Spark, rendererClasses []}]
 userVariables []
}
