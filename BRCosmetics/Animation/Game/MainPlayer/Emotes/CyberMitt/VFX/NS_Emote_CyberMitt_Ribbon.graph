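# Niagara ribbon trail system for the CyberMitt emote, with a single Lightning emitter on a short loop.
# Spawn rate feeds short-lived particles into a ribbon renderer, and a Leading_Edge_Fix_07 scratch module takes the emitter's simulation position, apparently to keep the trail's head attached.
# Over each particle's life, curves fade alpha, shift color between purple and white, and widen the ribbon from half to full width.
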NiagaraSystem /BRCosmetics/Animation/Game/MainPlayer/Emotes/CyberMitt/VFX/NS_Emote_CyberMitt_Ribbon
dependencies {
 usedDataInterfaces [&NiagaraDataInterfaceColorCurve, &NiagaraDataInterfaceCurve]
 usedDynamicInputs []
 usedModules [&SystemState]
}
emitters {Lightning {
  emitterUpdateScript [
   &EmitterState(Inactive Response=::NewEnumerator1, Life Cycle Mode=::NewEnumerator1, Loop Behavior=::NewEnumerator1, Loop Duration=1.5, UseLoopDelay=false)
   /SetVariables(Emitter.NewSpawnRate=100) {isSetParameters=true}
   &SpawnRate(SpawnRate=$Emitter.NewSpawnRate, Use Spawn Probability=false)
  ]
  particleSpawnScript [&InitializeParticle(
    Lifetime=0.23
    Ribbon Facing Mode=::NewEnumerator1
    Ribbon Width=30
    Ribbon Width Mode=::NewEnumerator1
    UsePositionOffset=false
    Write Material Random=true
   )]
  particleUpdateScript [
   &SolveForcesAndVelocity(Clamp Velocity=false, Limit Acceleration=false)
   &self:Leading_Edge_Fix_07(InputPosition:NiagaraPosition=&AddVectorToPosition(Position=$Engine.Emitter.SimulationPosition, Vector=<0 0 0>)) {module=Leading_Edge_Fix_07}
   &ParticleState()
   &ScaleColor(
    Scale Alpha=&FloatFromCurve(FloatCurve=DataInterface() {propertyValues {Curve {keys [<at 0 1 RCIM_Cubic>,<at 1 0>]}}})
    Scale RGB=&MakeVectorFromLinearColorRGB(Color=&ColorFromCurve(ColorCurve=DataInterface() {propertyValues {
        AlphaCurve {keys []}
        BlueCurve {keys [<at 0.2099 1 RCIM_Cubic>,<at 0.7377 1 RCIM_Cubic>]}
        GreenCurve {keys [<at 0.2099 0 RCIM_Cubic>,<at 0.7377 1 RCIM_Cubic>]}
        RedCurve {keys [<at 0.2099 0.5871 RCIM_Cubic>,<at 0.7377 1 RCIM_Cubic>]}
       }}))
    ScaleA=true
    ScaleRGB=true
   ) {module=ScaleColor002}
   /SetVariables(Particles.Glow Dyn Mat Param:Vector4f=&MakeVector4(X=1, Y=1, Z=1, W=1)) {isSetParameters=true}
   &ScaleRibbonWidth(Ribbon Width Scale=&FloatFromCurve(FloatCurve=DataInterface() {propertyValues {Curve {keys [
         {arriveTangent=1.7742, interpMode=RCIM_Cubic, leaveTangent=1.7742, tangentMode=RCTM_User, time=0, value=0.5}
         {arriveTangent=0.0224, interpMode=RCIM_Cubic, leaveTangent=0.0224, tangentMode=RCTM_User, time=1, value=1}
        ]}}}))
  ]
  rendererClasses [&NiagaraRibbonRendererProperties]
  renderers [{rendererClass=&NiagaraRibbonRendererProperties}]
 }}
summary {
 systemName=NS_Emote_CyberMitt_Ribbon
 emitters [{emitterName=Lightning, rendererClasses [&NiagaraRibbonRendererProperties]}]
 userVariables [
  User.bIsFrontEnd {type=NiagaraBool}
  User.bIsFrontEndPreview {type=NiagaraBool}
  User.bKillStage {type=NiagaraInt32}
  User.RibbonIDOffset {type=NiagaraInt32}
 ]
}
